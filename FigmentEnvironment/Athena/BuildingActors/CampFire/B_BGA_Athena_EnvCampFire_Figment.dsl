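; Figment variant of the Athena environment campfire building actor, a child Blueprint that decides when the campfire's first interact prompt is shown.
; ShouldDisplayFirstInteractPrompt returns true whenever the StokeEnabled scalable float evaluates true; otherwise it shows the prompt only while bHasBeenLitNative is false.
; The event graph holds only empty event stubs, and the construction script just calls its parent.
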(event EventBeginPlay
  (|CallParentFunction) ; This node is disabled and will not be called.\nDrag off pins to build functionality.
)

(event Collision|EventActorBeginOverlap
  (params (OtherActor :type Actor%20Object%20Reference)))

(event EventTick
  (params (DeltaSeconds :type Float%20%28single-precision%29)))

(event EventCustomBeginOverlap
  (params (InputPin :type Actor%20Object%20Reference)))

(fn ShouldDisplayFirstInteractPrompt
  (params (InteractingPawn :type Fort%20Pawn%20Object%20Reference) (InteractionType :type TInteractionType%20Enum :default "IT_NoInteraction"))
  (if (ScalableFloat|GetValueasBool (Variables|Default|GetStokeEnabled))
    (return true)
    (else
      (return (not (Variables|Default|GetbHasBeenLitNative))))))

(fn ConstructionScript
  (|CallParentFunction))
